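# Post-process style AnimBlueprint for the FunkyFeather Beak keytar cosmetic, adding secondary motion to its money, handle and rubber duck bones.
# AnimGraph offsets bones on the input pose, then chains spring-based Anim Dynamics nodes with per-bone limits and gravity.
# EventGraph updates duck sine-bob, sim, alpha and gravity variables from skydive, front-end and emote states.
# It also infers backpack or off-pawn placement from spine-to-handle socket distance.
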AnimBlueprint /SparksCosmetics/Instrument/Keytar/FunkyFeather_Beak/Meshes/FunkyFeather_Keytar_Beak_AnimBP
target_skeleton=&/FunkyFeather_Beak_Keytar_Skeleton
graphs [
 AnimGraph {nodes [
   n1 = Input Pose()
   n2 = "Transform (Modify) Bone - Bone: dyn_handle_01"!(
    ComponentPose="Transform (Modify) Bone - Bone: base_Money_Top_C"!(ComponentPose=Local To Component(LocalPose=@n1), Alpha=1, Translation="0.2,-1,0") -> (Pose) {translationMode=BMM_Additive, boneToModify {boneName=base_Money_Top_C}, translation=<0.2 -1>}
    Alpha=1
    Rotation="0,-78,0"
   ) -> (Pose) {
    rotationMode=BMM_Additive
    rotationSpace=BCS_BoneSpace
    scaleMode=BMM_Additive
    translationMode=BMM_Additive
    boneToModify {boneName=dyn_handle_01}
    rotation {yaw=-78}
   }
   n3 = Anim Dynamics - Bone: dyn_Money_Top_A_01!(ComponentPose=@n2, Alpha=1) -> (Pose) {
    angularSpringConstant=15000
    bAngularSpring=true
    bUseGravityOverride=true
    simulationSpace=RootRelative
    boundBone {boneName=dyn_Money_Top_A_01}
    componentAppliedLinearAccClamp=<3000 3000 3000>
    componentLinearAccScale=<0.7 0.7 0.7>
    componentLinearVelScale=<0.5 0.5 0.5>
    gravityOverride {z=-980}
    physicsBodyDefinitions [{
      boundBone {boneName=dyn_Money_Top_A_01}
      boxExtents=<20 20 20>
      constraintSetup {angularLimitsMax=<0 5 15>, angularLimitsMin=<0 -5 0>, angularTarget=<0 0 0>}
      localJointOffset=<2 0 0>
     }]
    simSpaceSettings {simSpaceAngularAlpha=1}
   }
   n4 = Anim Dynamics - Bone: dyn_Money_Top_A_02!(ComponentPose=@n3, Alpha=1) -> (Pose) {
    angularSpringConstant=15000
    bAngularSpring=true
    bUseGravityOverride=true
    simulationSpace=RootRelative
    boundBone {boneName=dyn_Money_Top_A_02}
    componentAppliedLinearAccClamp=<3000 3000 3000>
    componentLinearAccScale=<0.4 0.4 0.4>
    componentLinearVelScale=<0.4 0.4 0.4>
    gravityOverride {z=-980}
    physicsBodyDefinitions [{
      boundBone {boneName=dyn_Money_Top_A_02}
      boxExtents=<20 20 20>
      constraintSetup {angularLimitsMax=<0 5 7>, angularLimitsMin=<0 -5 -25>, angularTarget=<0 0 0>}
      localJointOffset=<8 0 0>
     }]
    simSpaceSettings {simSpaceAngularAlpha=1}
   }
   n5 = Anim Dynamics - Bone: dyn_Money_Top_B_01!(ComponentPose=@n4, Alpha=1) -> (Pose) {
    angularSpringConstant=15000
    bAngularSpring=true
    bUseGravityOverride=true
    simulationSpace=RootRelative
    boundBone {boneName=dyn_Money_Top_B_01}
    componentAppliedLinearAccClamp=<3000 3000 3000>
    componentLinearAccScale=<0.5 0.5 0.5>
    componentLinearVelScale=<0.5 0.5 0.5>
    gravityOverride {z=-980}
    physicsBodyDefinitions [{
      boundBone {boneName=dyn_Money_Top_B_01}
      boxExtents=<40 40 40>
      constraintSetup {angularLimitsMax=<0 0 0>, angularLimitsMin=<0 0 -15>, angularTarget=<0 0 0>}
      localJointOffset=<2 0 0>
     }]
    simSpaceSettings {simSpaceAngularAlpha=1}
   }
   n6 = Anim Dynamics - Bone: dyn_Money_Top_B_02!(ComponentPose=@n5, Alpha=1) -> (Pose) {
    angularSpringConstant=15000
    bAngularSpring=true
    bUseGravityOverride=true
    simulationSpace=RootRelative
    boundBone {boneName=dyn_Money_Top_B_02}
    componentAppliedLinearAccClamp=<3000 3000 3000>
    componentLinearAccScale=<0.5 0.5 0.5>
    componentLinearVelScale=<0.5 0.5 0.5>
    gravityOverride {z=-980}
    physicsBodyDefinitions [{
      boundBone {boneName=dyn_Money_Top_B_02}
      boxExtents=<20 20 20>
      constraintSetup {angularLimitsMax=<0 0 20>, angularLimitsMin=<0 0 -10>, angularTarget=<0 0 0>}
      localJointOffset=<8 0 0>
     }]
   }
   n7 = Anim Dynamics - Bone: dyn_Money_Top_C_01!(ComponentPose=@n6, Alpha=1) -> (Pose) {
    angularSpringConstant=15000
    bAngularSpring=true
    bUseGravityOverride=true
    simulationSpace=RootRelative
    boundBone {boneName=dyn_Money_Top_C_01}
    componentAppliedLinearAccClamp=<3000 3000 3000>
    componentLinearAccScale=<0.5 0.5 0.5>
    componentLinearVelScale=<0.5 0.5 0.5>
    gravityOverride {z=-980}
    physicsBodyDefinitions [{
      boundBone {boneName=dyn_Money_Top_C_01}
      boxExtents=<40 40 40>
      constraintSetup {angularLimitsMax=<0 5 10>, angularLimitsMin=<0 0 -15>, angularTarget=<0 0 0>}
      localJointOffset=<2 0 0>
     }]
    simSpaceSettings {simSpaceAngularAlpha=1}
   }
   n8 = Anim Dynamics - Bone: dyn_Money_Top_C_02!(ComponentPose=@n7, Alpha=1) -> (Pose) {
    angularSpringConstant=15000
    bAngularSpring=true
    bUseGravityOverride=true
    simulationSpace=RootRelative
    boundBone {boneName=dyn_Money_Top_C_02}
    componentAppliedLinearAccClamp=<3000 3000 3000>
    componentLinearAccScale=<0.5 0.5 0.5>
    componentLinearVelScale=<0.5 0.5 0.5>
    gravityOverride {z=-980}
    physicsBodyDefinitions [{
      boundBone {boneName=dyn_Money_Top_C_02}
      boxExtents=<15 15 15>
      constraintSetup {angularLimitsMax=<0 0 25>, angularLimitsMin=<0 0 0>, angularTarget=<0 0 0>}
      localJointOffset=<8 0 0>
     }]
   }
   n9 = Anim Dynamics - Bone: dyn_Money_Bot_D_01!(ComponentPose=@n8, Alpha=1) -> (Pose) {
    angularSpringConstant=15000
    bAngularSpring=true
    bUseGravityOverride=true
    simulationSpace=BoneRelative
    boundBone {boneName=dyn_Money_Bot_D_01}
    componentAppliedLinearAccClamp=<3000 3000 3000>
    componentLinearAccScale=<0.5 0.5 0.5>
    componentLinearVelScale=<0.5 0.5 0.5>
    gravityOverride {z=-980}
    physicsBodyDefinitions [{
      boundBone {boneName=dyn_Money_Bot_D_01}
      boxExtents=<20 20 20>
      constraintSetup {angularLimitsMax=<0 10 25>, angularLimitsMin=<0 -10 -25>, angularTarget=<0 0 0>}
      localJointOffset=<4 0 0>
     }]
    relativeSpaceBone {boneName=handle}
    simSpaceSettings {simSpaceAngularAlpha=1}
   }
   n10 = Anim Dynamics - Bone: dyn_Money_Bot_D_02!(ComponentPose=@n9, Alpha=1) -> (Pose) {
    angularSpringConstant=15000
    bAngularSpring=true
    bEnableWind=true
    bUseGravityOverride=true
    simulationSpace=RootRelative
    boundBone {boneName=dyn_Money_Bot_D_02}
    componentLinearAccScale=<0.2 0.2 0.2>
    gravityOverride {z=-980}
    physicsBodyDefinitions [{
      boundBone {boneName=dyn_Money_Bot_D_02}
      boxExtents=<20 20 20>
      constraintSetup {angularLimitsMax=<0 0 50>, angularLimitsMin=<0 0 -50>, angularTarget=<0 0 0>}
      localJointOffset=<6 0 0>
     }]
    simSpaceSettings {simSpaceAngularAlpha=1}
   }
   n11 = Anim Dynamics - Bone: dyn_Money_Bot_E_01!(ComponentPose=@n10, Alpha=1) -> (Pose) {
    angularSpringConstant=25000
    bAngularSpring=true
    bUseGravityOverride=true
    simulationSpace=RootRelative
    boundBone {boneName=dyn_Money_Bot_E_01}
    componentAppliedLinearAccClamp=<3000 3000 3000>
    componentLinearAccScale=<0.5 0.5 0.5>
    componentLinearVelScale=<0.5 0.5 0.5>
    gravityOverride {z=-1500}
    physicsBodyDefinitions [{
      boundBone {boneName=dyn_Money_Bot_E_01}
      boxExtents=<20 20 20>
      constraintSetup {angularLimitsMax=<0 10 15>, angularLimitsMin=<0 -10 -15>, angularTarget=<0 0 0>}
      localJointOffset=<8 0 0>
     }]
    simSpaceSettings {simSpaceAngularAlpha=1}
   }
   n12 = Anim Dynamics - Bone: dyn_Money_Bot_E_02!(ComponentPose=@n11, Alpha=1) -> (Pose) {
    angularSpringConstant=15000
    bAngularSpring=true
    bUseGravityOverride=true
    simulationSpace=RootRelative
    boundBone {boneName=dyn_Money_Bot_E_02}
    gravityOverride {z=-980}
    physicsBodyDefinitions [{
      boundBone {boneName=dyn_Money_Bot_E_02}
      boxExtents=<15 15 15>
      constraintSetup {angularLimitsMax=<0 10 10>, angularLimitsMin=<0 -10 -10>, angularTarget=<0 0 0>}
      localJointOffset=<2 0 0>
     }]
    simSpaceSettings {simSpaceAngularAlpha=1}
   }
   n13 = Anim Dynamics - Bone: dyn_handle_01!(ComponentPose=@n12, Alpha=1, GravityOverride:Vector="0,0,-4000") -> (Pose) {
    angularDampingOverride=0.9
    angularSpringConstant=15000
    bAngularSpring=true
    bOverrideAngularDamping=true
    bUseGravityOverride=true
    simulationSpace=RootRelative
    boundBone {boneName=dyn_handle_01}
    componentAppliedLinearAccClamp=<3000 3000 3000>
    componentLinearAccScale=<0.5 0.5 0.5>
    componentLinearVelScale=<0.5 0.5 0.5>
    gravityOverride {z=-4000}
    physicsBodyDefinitions [{
      boundBone {boneName=dyn_handle_01}
      boxExtents=<40 40 40>
      constraintSetup {angularLimitsMax=<0 0 120>, angularLimitsMin=<0 0 0>, angularTarget=<0 0 0>}
      localJointOffset=<5 0 0>
     }]
    simSpaceSettings {maxAngularAcceleration=32, maxAngularVelocity=8, simSpaceAngularAlpha=1}
   }
   n14 = "Transform (Modify) Bone - Bone: dyn_rubberDuck_02"!(ComponentPose=@n13, Alpha=1, Translation="0,0,0", Rotation="-17.794,-5.318,85.000") -> (Pose) {
    rotationMode=BMM_Replace
    scaleMode=BMM_Replace
    translationMode=BMM_Additive
    boneToModify {boneName=dyn_rubberDuck_02}
    rotation=<rot -17.794 -5.318 85>
   }
   n15 = Anim Dynamics - Bone: dyn_rubberDuck_02!(
    ComponentPose=@n14
    Alpha=0.8
    SimSpaceSettings:Anim Phys Sim Space Settings Structure=(SimSpaceAngularAlpha=0.5,MaxAngularVelocity=8,MaxAngularAcceleration=28,ExternalAngularVelocity=(X=0,Y=0,Z=0))
   ) -> (Pose) {
    angularBiasOverride=0.99
    angularDampingOverride=0.99
    angularSpringConstant=10
    bLinearSpring=true
    bOverrideLinearDamping=true
    bUseGravityOverride=true
    gravityScale=0.15
    linearDampingOverride=0.99
    linearSpringConstant=40
    simulationSpace=RootRelative
    alphaBoolBlend {blendOption=Linear}
    alphaScaleBiasClamp {bInterpResult=true, interpSpeedDecreasing=5, interpSpeedIncreasing=5}
    boundBone {boneName=dyn_rubberDuck_02}
    componentAppliedLinearAccClamp=<40 40 40>
    componentLinearAccScale=<0.5 0.5 0.5>
    componentLinearVelScale=<0.2 0.2 0.1>
    physicsBodyDefinitions [{
      boundBone {boneName=dyn_rubberDuck_02}
      boxExtents=<44 44 44>
      constraintSetup {angularLimitsMax=<0 0 10>, angularLimitsMin=<0 0 -10>, angularTarget=<1 0 0>, linearAxesMax=<3 5 0>, linearAxesMin=<-3 -10 0>}
      localJointOffset=<0 2 0>
     }]
    simSpaceSettings {maxAngularAcceleration=28, maxAngularVelocity=8, simSpaceAngularAlpha=0.5}
   }
   n16 = "Transform (Modify) Bone - Bone: dyn_rubberDuck_02"(
    ComponentPose="Transform (Modify) Bone - Bone: dyn_rubberDuck_02"(ComponentPose=@n15, Alpha=1, Translation="0,0,0", Rotation="-100,0,0") {rotationMode=BMM_Additive, scaleMode=BMM_Replace, boneToModify {boneName=dyn_rubberDuck_02}, rotation {pitch=-100}}
    Alpha=1
    Translation="-2,0,4"
    Rotation="-180,0,0"
   ) {
    rotationMode=BMM_Additive
    scaleMode=BMM_Replace
    translationMode=BMM_Additive
    boneToModify {boneName=dyn_rubberDuck_02}
    rotation {pitch=-180}
    translation {x=-2, z=4}
   }
   Output Pose(Result=Component To Local(ComponentPose=@n16))
  ]}
 EventGraph {nodes [
   Event BlueprintUpdateAnimation()
   > n1 = Sequence!(execute) -> (then_0, then_1, then_2:Exec, then_3:Exec)
   > n2 = Branch(Condition=Get bIsSkyDiving())
   > Set DuckSimSettings(DuckSimSettings:Anim Phys Sim Space Settings Structure=Get DuckSimSkyDiving() -> (DuckSimSkyDiving:Anim Phys Sim Space Settings Structure)) -> (Output_Get:Anim Phys Sim Space Settings Structure)
   > Set DuckPush(DuckPush:double=-1) -> (Output_Get:double)
   Set DuckSimSettings(execute=@n2.else, DuckSimSettings=Get DuckGround() -> (DuckGround:Anim Phys Sim Space Settings Structure))
   > Set DuckPush(DuckPush=0)
   n3 = float * float(A=GetGameTimeInSeconds(), B=4) {is=title K2Node_PromotableOperator}
   n4 = float * float(A="Sin (Radians)"(A=float * float(A=0.16, B=@n3)), B=30) {is=title K2Node_CommutativeAssociativeBinaryOperator}
   n5 = float * float(A="Sin (Radians)"(A=float * float(A=0.16, B=@n3)), B=0.32)
   Set Duck_1 Rotator(execute=@n1.then_2, Duck_1 Rotator:Rotator=MakeRotator(Roll=85, Pitch=float + float(A=@n4, B=95), Yaw=float + float(A=@n5, B=-5))) -> (Output_Get:Rotator)
   n6 = float * float(A=GetGameTimeInSeconds(), B=3) {is=title K2Node_PromotableOperator}
   n7 = "Sin (Radians)"(A=float * float(A=0.5, B=@n6) {is=title K2Node_CommutativeAssociativeBinaryOperator})
   n8 = float * float(A="Sin (Radians)"(A=float * float(A=0.4, B=@n6)), B=0.9)
   > Set Duck_1 Floating(Duck_1 Floating:Vector=MakeVector(X=float + float(A=@n8, B=4), Y=Get DuckPush() -> (DuckPush:double), Z=float * float(A=@n7, B=0.3))) -> (Output_Get:Vector)
   n9 = Branch(execute=@n1.then_1, Condition=Get bIsInFrontEnd())
   > Set DuckAlpha(DuckAlpha:double=0.5) -> (Output_Get:double)
   > Set AlphaBills(AlphaBills:double=0.6) -> (Output_Get:double)
   > Set Override World Gravity(Override World Gravity:Vector="0,0,-4100") -> (Output_Get:Vector)
   > n10 = Branch(Condition=Get bIsPlayingEmote())
   Set HandleAlpha(execute=@n10.else, HandleAlpha:double=0.88) -> (Output_Get:double)
   Set DuckAlpha(execute=@n9.else, DuckAlpha=0.8)
   > Set AlphaBills(AlphaBills=1)
   > Set Override World Gravity(Override World Gravity="0,0,-1500")
   n11 = GetPartSkeletalMeshComponent()
   Is Valid(exec=@n1.then_3, InputObject=@n11)
   n12 = GetAttachParent!(self:Scene Component Object Reference=@n11) -> (ReturnValue:Scene Component Object Reference)
   > Is Valid(InputObject=@n12)
   > n13 = Set Body_Mesh(Body_Mesh:Scene Component Object Reference=@n12) -> (Output_Get:Scene Component Object Reference)
   n14 = GetOwningComponent()
   n15 = "Distance (Vector)"!(
    V1:Vector=BreakTransform(InTransform=GetSocketTransform(self=@n13, InSocketName=spine_05, TransformSpace=RTS_World))
    V2:Vector=BreakTransform(InTransform=GetSocketTransform(self=@n14, InSocketName=base_handle, TransformSpace=RTS_World))
   ) -> (ReturnValue:double)
   > Set Backpack?(Backpack?:bool="float < float"(A=@n15, B=41)) -> (Output_Get:bool)
   n16 = AND Boolean(A=Get bIsInFrontEnd(), B=NOT Boolean(A=Get bIsOnPawn() -> (bIsOnPawn:bool)))
   n17 = "float < float"(A=@n15, B=50)
   > Set KeytarOffPawn(KeytarOffPawn:bool=AND Boolean(A=@n17, B=@n16)) -> (Output_Get:bool)
   > Set PickaxeOffPawn(PickaxeOffPawn:bool=AND Boolean(A=NOT Boolean(A=@n17), B=@n16)) -> (Output_Get:bool)
   Event OnBeginPlayingEmote() {is=title K2Node_Event}
   > Set HandleAlpha(HandleAlpha=0)
   Event OnEndPlayingEmote() {is=title K2Node_Event}
   > Set HandleAlpha(HandleAlpha=1)
  ]}
]
variables [
 AlphaBills {type=number}
 Backpack? {type=boolean}
 Body_Mesh {type=SceneComponent, default=null}
 Duck_1 Floating {type=Vector}
 Duck_1 Rotator {type=Rotator}
 DuckAlpha {type=number}
 DuckGround {
  type=AnimPhysSimSpaceSettings
  default {maxAngularAcceleration=32, maxAngularVelocity=8, simSpaceAngularAlpha=0.6, externalAngularVelocity=<0 0 0>}
 }
 DuckPush {type=number}
 DuckSimSettings {
  type=AnimPhysSimSpaceSettings
  default {maxAngularAcceleration=32, maxAngularVelocity=8, simSpaceAngularAlpha=0.6, externalAngularVelocity=<0 0 0>}
 }
 DuckSimSkyDiving {
  type=AnimPhysSimSpaceSettings
  default {maxAngularAcceleration=16, maxAngularVelocity=4, simSpaceAngularAlpha=0.5, externalAngularVelocity=<0 0 0>}
 }
 HandleAlpha {type=number, default=1}
 HandleOffsetAlpha {type=number}
 KeytarOffPawn {type=boolean}
 Override World Gravity {type=Vector}
 PickaxeOffPawn {type=boolean}
]
